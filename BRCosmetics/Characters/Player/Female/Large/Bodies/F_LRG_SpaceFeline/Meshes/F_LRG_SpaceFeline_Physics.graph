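# Physics asset for the large female SpaceFeline player body, with capsule bodies for the torso, arms, thighs, jacket and two tube bones.
# Jacket back, side and front bones form left and right chains hanging from spine_04, with tight swing limits and zero twist limits.
# Torso, limb and tube constraints leave angular motion free.
# Two mirrored free constraints also cross-link the left and right back jacket tips.
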PhysicsAsset /BRCosmetics/Characters/Player/Female/Large/Bodies/F_LRG_SpaceFeline/Meshes/F_LRG_SpaceFeline_Physics
bodies {
 jacket_bck_l_01 [{shapeType=Capsule, length=5.5, radius=4.5, shapeName=jacket_bck_l_01_capsule, center=<1.908 0 0>, rotation=<rot -0.0005 0.0001 90>}]
 jacket_bck_l_02 [{shapeType=Capsule, length=6, radius=3.5, shapeName=jacket_bck_l_02_capsule, center=<1.908 0 0>, rotation=<rot -0.0005 0.0001 90>}]
 jacket_bck_r_01 [{shapeType=Capsule, length=6, radius=4, shapeName=jacket_bck_r_01_capsule, center=<1.908 0 0>, rotation=<rot -0.0005 -0.0001 90>}]
 jacket_bck_r_02 [{shapeType=Capsule, length=6, radius=4, shapeName=jacket_bck_r_02_capsule, center=<1.908 0 0>, rotation=<rot -0.0005 -0.0001 90>}]
 jacket_fr_l_01 [{
   shapeType=Capsule
   length=6
   radius=2
   shapeName=jacket_fr_l_01_capsule
   center=<4.2847 0.0443 -0.2549>
   rotation=<rot 19.6835 89.3736 86.3839>
  }]
 jacket_fr_l_02 [{
   shapeType=Capsule
   length=5.5
   radius=2.5
   shapeName=jacket_fr_l_02_capsule
   center=<4.2947 0.3024 -0.9765>
   rotation=<rot 19.6837 89.3737 96.384>
  }]
 jacket_fr_l_03 [{
   shapeType=Capsule
   length=5.5
   radius=2
   shapeName=jacket_fr_l_03_capsule
   center=<4.3535 -0.0309 -0.0468>
   rotation=<rot 19.6837 89.3736 96.384>
  }]
 jacket_fr_r_01 [{
   shapeType=Capsule
   length=5.5
   radius=2
   shapeName=jacket_fr_r_01_capsule
   center=<4.2847 0.0442 0.255>
   rotation=<rot -19.6835 89.3714 93.6167>
  }]
 jacket_fr_r_02 [{
   shapeType=Capsule
   length=6
   radius=2
   shapeName=jacket_fr_r_02_capsule
   center=<4.2847 0.0442 0.255>
   rotation=<rot -19.6837 89.3714 93.6167>
  }]
 jacket_fr_r_03 [{
   shapeType=Capsule
   length=5
   radius=3
   shapeName=jacket_fr_r_03_capsule
   center=<5.4531 0.2062 0.7437>
   rotation=<rot -19.6837 89.3714 93.6167>
  }]
 jacket_side_l_01 [{
   shapeType=Capsule
   length=11.5
   radius=2.5
   shapeName=jacket_side_l_01_capsule
   center=<3.3775 -0.6966 0.5498>
   rotation=<rot -0.0001 -19.9995 90>
  }]
 jacket_side_l_02 [{
   shapeType=Capsule
   length=11
   radius=2.5
   shapeName=jacket_side_l_02_capsule
   center=<3.3775 -0.6966 0.4952>
   rotation=<rot -0.0001 -19.9995 90>
  }]
 jacket_side_r_01 [{
   shapeType=Capsule
   length=10.5
   radius=3
   shapeName=jacket_side_r_01_capsule
   center=<4.5316 0.5235 -0.4027>
   rotation=<rot -0.0001 19.9995 90>
  }]
 jacket_side_r_02 [{shapeType=Capsule, length=13.5, radius=2, shapeName=jacket_side_r_02_capsule, center=<4.4464 0.2758 0>, rotation=<rot -0.0001 19.9995 90>}]
 lowerarm_l [{
   shapeType=Capsule
   length=19.316
   radius=4.1552
   shapeName=lowerarm_l_capsule
   center=<7.4867 -0.9759 0.6031>
   rotation=<rot 0 89.4439 -86.875>
  }]
 lowerarm_r [{
   shapeType=Capsule
   length=22.0642
   radius=6.6227
   shapeName=lowerarm_r_capsule
   center=<-7.3187 0.8957 -0.6345>
   rotation=<rot 0 89.3231 -86.0689>
  }]
 pelvis [{shapeType=Capsule, length=8, radius=13, shapeName=pelvis_capsule, center=<0.7345 -0.6834 0>}]
 spine_01 [{shapeType=Capsule, length=5, radius=3, shapeName=spine_01_capsule}]
 spine_02 [{shapeType=Capsule, length=5, radius=8, shapeName=spine_02_capsule}]
 spine_03 [{shapeType=Capsule, length=3.5, radius=9, shapeName=spine_03_capsule, center=<1.9598 -0.2785 0>}]
 spine_04 [{shapeType=Capsule, length=4.5, radius=8, shapeName=spine_04_capsule, center=<0 2.0467 0>}]
 thigh_l [{shapeType=Capsule, length=24.5, radius=9.5, shapeName=thigh_l_capsule, center=<-12.499 0 0>, rotation=<rot -90 -179.9963 179.9962>}]
 thigh_r [{shapeType=Capsule, length=24.5, radius=9.5, shapeName=thigh_r_capsule, center=<12.499 -0.0001 0>, rotation=<rot 90 179.9999 -179.9998>}]
 tube_l_02 [{
   shapeType=Capsule
   length=9.4011
   radius=1.6217
   shapeName=tube_l_02_capsule
   center=<1.3196 0.2699 -0.6026>
   rotation=<rot 0 4.0202 -86.6141>
  }]
 tube_r_02 [{
   shapeType=Capsule
   length=7.2892
   radius=1.8199
   shapeName=tube_r_02_capsule
   center=<1.1525 0.6045 -0.1201>
   rotation=<rot 0 -173.0544 -88.642>
  }]
 upperarm_l [{shapeType=Capsule, length=21.5, radius=5, shapeName=upperarm_l_capsule, center=<12.2639 0.0001 2.376>, rotation=<rot 80 0 -0.0006>}]
 upperarm_r [{shapeType=Capsule, length=14.5, radius=6, shapeName=upperarm_r_capsule, center=<-10.1086 0 -1.982>, rotation=<rot -80 179.9995 -179.9994>}]
}
constraints [
 {
  bone1Name=spine_01
  bone2Name=pelvis
  swing1LimitDegrees=45
  swing1Motion=Free
  swing2LimitDegrees=45
  swing2Motion=Free
  twistLimitDegrees=45
  twistMotion=Free
 }
 {
  bone1Name=spine_02
  bone2Name=spine_01
  swing1LimitDegrees=45
  swing1Motion=Free
  swing2LimitDegrees=45
  swing2Motion=Free
  twistLimitDegrees=45
  twistMotion=Free
 }
 {
  bone1Name=spine_03
  bone2Name=spine_02
  swing1LimitDegrees=45
  swing1Motion=Free
  swing2LimitDegrees=45
  swing2Motion=Free
  twistLimitDegrees=45
  twistMotion=Free
 }
 {
  bone1Name=spine_04
  bone2Name=spine_03
  swing1LimitDegrees=45
  swing1Motion=Free
  swing2LimitDegrees=45
  swing2Motion=Free
  twistLimitDegrees=45
  twistMotion=Free
 }
 {
  bone1Name=upperarm_l
  bone2Name=spine_04
  swing1LimitDegrees=45
  swing1Motion=Free
  swing2LimitDegrees=45
  swing2Motion=Free
  twistLimitDegrees=45
  twistMotion=Free
 }
 {
  bone1Name=upperarm_r
  bone2Name=spine_04
  swing1LimitDegrees=45
  swing1Motion=Free
  swing2LimitDegrees=45
  swing2Motion=Free
  twistLimitDegrees=45
  twistMotion=Free
 }
 {bone1Name=jacket_bck_l_01, bone2Name=spine_04, swing1LimitDegrees=10, swing2LimitDegrees=35, twistLimitDegrees=0}
 {bone1Name=jacket_bck_l_02, bone2Name=jacket_bck_l_01, swing1LimitDegrees=10, swing2LimitDegrees=35, twistLimitDegrees=0}
 {bone1Name=jacket_bck_r_01, bone2Name=spine_04, swing1LimitDegrees=10, swing2LimitDegrees=25, twistLimitDegrees=0}
 {bone1Name=jacket_bck_r_02, bone2Name=jacket_bck_r_01, swing1LimitDegrees=10, swing2LimitDegrees=35, twistLimitDegrees=0}
 {bone1Name=jacket_fr_l_01, bone2Name=spine_04, swing1LimitDegrees=5, swing2LimitDegrees=20, twistLimitDegrees=0}
 {bone1Name=jacket_fr_l_02, bone2Name=jacket_fr_l_01, swing1LimitDegrees=10, swing2LimitDegrees=35, twistLimitDegrees=0}
 {bone1Name=jacket_fr_l_03, bone2Name=jacket_fr_l_02, swing1LimitDegrees=10, swing2LimitDegrees=35, twistLimitDegrees=0}
 {bone1Name=jacket_fr_r_01, bone2Name=spine_04, swing1LimitDegrees=5, swing2LimitDegrees=30, twistLimitDegrees=0}
 {bone1Name=jacket_fr_r_02, bone2Name=jacket_fr_r_01, swing1LimitDegrees=5, swing2LimitDegrees=25, twistLimitDegrees=0}
 {bone1Name=jacket_fr_r_03, bone2Name=jacket_fr_r_02, swing1LimitDegrees=5, swing2LimitDegrees=25, twistLimitDegrees=0}
 {bone1Name=jacket_side_l_01, bone2Name=spine_04, swing1LimitDegrees=10, swing2LimitDegrees=45, twistLimitDegrees=0}
 {bone1Name=jacket_side_l_02, bone2Name=jacket_side_l_01, swing1LimitDegrees=10, swing2LimitDegrees=45, twistLimitDegrees=0}
 {bone1Name=jacket_side_r_01, bone2Name=spine_04, swing1LimitDegrees=10, swing2LimitDegrees=30, twistLimitDegrees=0}
 {bone1Name=jacket_side_r_02, bone2Name=jacket_side_r_01, swing1LimitDegrees=10, swing2LimitDegrees=30, twistLimitDegrees=0}
 {
  bone1Name=thigh_l
  bone2Name=pelvis
  swing1LimitDegrees=45
  swing1Motion=Free
  swing2LimitDegrees=45
  swing2Motion=Free
  twistLimitDegrees=45
  twistMotion=Free
 }
 {
  bone1Name=thigh_r
  bone2Name=pelvis
  swing1LimitDegrees=45
  swing1Motion=Free
  swing2LimitDegrees=45
  swing2Motion=Free
  twistLimitDegrees=45
  twistMotion=Free
 }
 {
  bone1Name=lowerarm_l
  bone2Name=upperarm_l
  swing1LimitDegrees=45
  swing1Motion=Free
  swing2LimitDegrees=45
  swing2Motion=Free
  twistLimitDegrees=45
  twistMotion=Free
 }
 {
  bone1Name=lowerarm_r
  bone2Name=upperarm_r
  swing1LimitDegrees=45
  swing1Motion=Free
  swing2LimitDegrees=45
  swing2Motion=Free
  twistLimitDegrees=45
  twistMotion=Free
 }
 {
  bone1Name=jacket_bck_l_02
  bone2Name=jacket_bck_r_02
  swing1LimitDegrees=45
  swing1Motion=Free
  swing2LimitDegrees=45
  swing2Motion=Free
  twistLimitDegrees=45
  twistMotion=Free
 }
 {
  bone1Name=jacket_bck_r_02
  bone2Name=jacket_bck_l_02
  swing1LimitDegrees=45
  swing1Motion=Free
  swing2LimitDegrees=45
  swing2Motion=Free
  twistLimitDegrees=45
  twistMotion=Free
 }
 {
  bone1Name=tube_l_02
  bone2Name=spine_04
  swing1LimitDegrees=45
  swing1Motion=Free
  swing2LimitDegrees=45
  swing2Motion=Free
  twistLimitDegrees=45
  twistMotion=Free
 }
 {
  bone1Name=tube_r_02
  bone2Name=spine_04
  swing1LimitDegrees=45
  swing1Motion=Free
  swing2LimitDegrees=45
  swing2Motion=Free
  twistLimitDegrees=45
  twistMotion=Free
 }
]
