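# Secondary-motion AnimBP for the large female SpaceFeline player body, layered on a pose copied from the attached parent mesh.
# A deform Control Rig, fed the front-end flag, runs first, then an additive tube_r_01 offset and half-alpha spring AnimDynamics on both tubes.
# A RigidBody node using the body's physics asset, simulated relative to spine_05, finishes the pose.
# Jacket AnimDynamics and knee constraint chains exist but are disconnected from the output.
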AnimBlueprint /BRCosmetics/Characters/Player/Female/Large/Bodies/F_LRG_SpaceFeline/Meshes/F_LRG_SpaceFeline_AnimBP
target_skeleton=/Game/Characters/Player/Female/Large/Bodies/F_LRG_SpaceFeline/Meshes/F_LRG_SpaceFeline_Skeleton
graphs [
 AnimGraph {nodes [
   n1 = Anim Dynamics - Bone: jacket_fr_r_01(Alpha=1, ExternalForce="0,0,0") {
    angularSpringConstant=2000
    bAngularSpring=true
    bGravityOverrideInSimSpace=true
    bUseGravityOverride=true
    simulationSpace=RootRelative
    alphaBoolBlend {blendOption=Linear}
    boundBone {boneName=jacket_fr_r_01}
    componentLinearAccScale=<1 1 1>
    componentLinearVelScale=<1 1 1>
    gravityOverride {z=-1500}
    physicsBodyDefinitions [{
      boundBone {boneName=jacket_fr_r_01}
      boxExtents=<30 30 30>
      constraintSetup {angularLimitsMax=<0 5 10>, angularLimitsMin=<0 -20 -10>, angularTarget=<0 0 0>}
      localJointOffset=<3 0 0>
     }]
    relativeSpaceBone {boneName=root}
   }
   n2 = Anim Dynamics - Bone: jacket_fr_r_02(ComponentPose=@n1, Alpha=1, ExternalForce="0,0,0") {
    angularSpringConstant=2000
    bAngularSpring=true
    bGravityOverrideInSimSpace=true
    bUseGravityOverride=true
    simulationSpace=RootRelative
    alphaBoolBlend {blendOption=Linear}
    boundBone {boneName=jacket_fr_r_02}
    componentLinearAccScale=<1 1 1>
    componentLinearVelScale=<1 1 1>
    gravityOverride {z=-1500}
    physicsBodyDefinitions [{
      boundBone {boneName=jacket_fr_r_02}
      boxExtents=<30 30 30>
      constraintSetup {angularLimitsMax=<0 5 10>, angularLimitsMin=<0 -45 -10>, angularTarget=<0 0 0>}
      localJointOffset=<3 0 0>
     }]
    relativeSpaceBone {boneName=root}
   }
   n3 = Anim Dynamics - Bone: jacket_fr_r_03(ComponentPose=@n2, Alpha=1, ExternalForce="0,0,0") {
    angularSpringConstant=2000
    bAngularSpring=true
    bGravityOverrideInSimSpace=true
    bUseGravityOverride=true
    simulationSpace=RootRelative
    alphaBoolBlend {blendOption=Linear}
    boundBone {boneName=jacket_fr_r_03}
    componentLinearAccScale=<1 1 1>
    componentLinearVelScale=<1 1 1>
    gravityOverride {z=-1500}
    physicsBodyDefinitions [{
      boundBone {boneName=jacket_fr_r_03}
      boxExtents=<30 30 30>
      constraintSetup {angularLimitsMax=<0 5 10>, angularLimitsMin=<0 -50 -10>, angularTarget=<0 0 0>}
      localJointOffset=<3 0 0>
     }]
    relativeSpaceBone {boneName=root}
   }
   n4 = Anim Dynamics - Bone: jacket_fr_l_01(ComponentPose=@n3, Alpha=1, ExternalForce="0,0,0") {
    angularSpringConstant=2000
    bAngularSpring=true
    bGravityOverrideInSimSpace=true
    bUseGravityOverride=true
    simulationSpace=RootRelative
    alphaBoolBlend {blendOption=Linear}
    boundBone {boneName=jacket_fr_l_01}
    componentLinearAccScale=<1 1 1>
    componentLinearVelScale=<1 1 1>
    gravityOverride {z=-1500}
    physicsBodyDefinitions [{
      boundBone {boneName=jacket_fr_l_01}
      boxExtents=<30 30 30>
      constraintSetup {angularLimitsMax=<0 5 10>, angularLimitsMin=<0 -20 -10>, angularTarget=<0 0 0>}
      localJointOffset=<3 0 0>
     }]
    relativeSpaceBone {boneName=root}
   }
   n5 = Anim Dynamics - Bone: jacket_fr_l_02(ComponentPose=@n4, Alpha=1, ExternalForce="0,0,0") {
    angularSpringConstant=2000
    bAngularSpring=true
    bGravityOverrideInSimSpace=true
    bUseGravityOverride=true
    simulationSpace=RootRelative
    alphaBoolBlend {blendOption=Linear}
    boundBone {boneName=jacket_fr_l_02}
    componentLinearAccScale=<1 1 1>
    componentLinearVelScale=<1 1 1>
    gravityOverride {z=-1500}
    physicsBodyDefinitions [{
      boundBone {boneName=jacket_fr_l_02}
      boxExtents=<30 30 30>
      constraintSetup {angularLimitsMax=<0 5 10>, angularLimitsMin=<0 -45 -10>, angularTarget=<0 0 0>}
      localJointOffset=<3 0 0>
     }]
    relativeSpaceBone {boneName=root}
   }
   n6 = Anim Dynamics - Bone: jacket_fr_l_03(ComponentPose=@n5, Alpha=1, ExternalForce="0,0,0") {
    angularSpringConstant=2000
    bAngularSpring=true
    bGravityOverrideInSimSpace=true
    bUseGravityOverride=true
    simulationSpace=RootRelative
    alphaBoolBlend {blendOption=Linear}
    boundBone {boneName=jacket_fr_l_03}
    componentLinearAccScale=<1 1 1>
    componentLinearVelScale=<1 1 1>
    gravityOverride {z=-1500}
    physicsBodyDefinitions [{
      boundBone {boneName=jacket_fr_l_03}
      boxExtents=<30 30 30>
      constraintSetup {angularLimitsMax=<0 5 10>, angularLimitsMin=<0 -50 -10>, angularTarget=<0 0 0>}
      localJointOffset=<3 0 0>
     }]
    relativeSpaceBone {boneName=root}
   }
   n7 = Anim Dynamics - Bone: jacket_side_r_01(ComponentPose=@n6, Alpha=1, ExternalForce="0,0,0") {
    angularSpringConstant=2000
    bAngularSpring=true
    bGravityOverrideInSimSpace=true
    bUseGravityOverride=true
    simulationSpace=RootRelative
    alphaBoolBlend {blendOption=Linear}
    boundBone {boneName=jacket_side_r_01}
    componentLinearAccScale=<1 1 1>
    componentLinearVelScale=<1 1 1>
    gravityOverride {z=-1500}
    physicsBodyDefinitions [{
      boundBone {boneName=jacket_side_r_01}
      boxExtents=<30 30 30>
      constraintSetup {angularLimitsMax=<0 5 10>, angularLimitsMin=<0 -45 -10>, angularTarget=<0 0 0>}
      localJointOffset=<3 0 0>
     }]
    relativeSpaceBone {boneName=root}
   }
   n8 = Anim Dynamics - Bone: jacket_side_r_02(ComponentPose=@n7, Alpha=1, ExternalForce="0,0,0") {
    angularSpringConstant=2000
    bAngularSpring=true
    bGravityOverrideInSimSpace=true
    bUseGravityOverride=true
    simulationSpace=RootRelative
    alphaBoolBlend {blendOption=Linear}
    boundBone {boneName=jacket_side_r_02}
    componentLinearAccScale=<1 1 1>
    componentLinearVelScale=<1 1 1>
    gravityOverride {z=-1500}
    physicsBodyDefinitions [{
      boundBone {boneName=jacket_side_r_02}
      boxExtents=<30 30 30>
      constraintSetup {angularLimitsMax=<0 5 10>, angularLimitsMin=<0 -50 -10>, angularTarget=<0 0 0>}
      localJointOffset=<3 0 0>
     }]
    relativeSpaceBone {boneName=root}
   }
   n9 = Anim Dynamics - Bone: jacket_bck_r_01(ComponentPose=@n8, Alpha=1, ExternalForce="0,0,0") {
    angularSpringConstant=2000
    bAngularSpring=true
    bGravityOverrideInSimSpace=true
    bUseGravityOverride=true
    simulationSpace=RootRelative
    alphaBoolBlend {blendOption=Linear}
    boundBone {boneName=jacket_bck_r_01}
    componentLinearAccScale=<1 1 1>
    componentLinearVelScale=<1 1 1>
    gravityOverride {z=-1500}
    physicsBodyDefinitions [{
      boundBone {boneName=jacket_bck_r_01}
      boxExtents=<30 30 30>
      constraintSetup {angularLimitsMax=<0 5 10>, angularLimitsMin=<0 -30 -10>, angularTarget=<0 0 0>}
      localJointOffset=<3 0 0>
     }]
    relativeSpaceBone {boneName=root}
   }
   n10 = Anim Dynamics - Bone: jacket_bck_r_02(ComponentPose=@n9, Alpha=1, ExternalForce="0,0,0") {
    angularSpringConstant=2000
    bAngularSpring=true
    bGravityOverrideInSimSpace=true
    bUseGravityOverride=true
    simulationSpace=RootRelative
    alphaBoolBlend {blendOption=Linear}
    boundBone {boneName=jacket_bck_r_02}
    componentLinearAccScale=<1 1 1>
    componentLinearVelScale=<1 1 1>
    gravityOverride {z=-1500}
    physicsBodyDefinitions [{
      boundBone {boneName=jacket_bck_r_02}
      boxExtents=<30 30 30>
      constraintSetup {angularLimitsMax=<0 5 10>, angularLimitsMin=<0 -50 -10>, angularTarget=<0 0 0>}
      localJointOffset=<3 0 0>
     }]
    relativeSpaceBone {boneName=root}
   }
   n11 = Anim Dynamics - Bone: jacket_bck_l_01(ComponentPose=@n10, Alpha=1, ExternalForce="0,0,0") {
    angularSpringConstant=2000
    bAngularSpring=true
    bGravityOverrideInSimSpace=true
    bUseGravityOverride=true
    simulationSpace=RootRelative
    alphaBoolBlend {blendOption=Linear}
    boundBone {boneName=jacket_bck_l_01}
    componentLinearAccScale=<1 1 1>
    componentLinearVelScale=<1 1 1>
    gravityOverride {z=-1500}
    physicsBodyDefinitions [{
      boundBone {boneName=jacket_bck_l_01}
      boxExtents=<30 30 30>
      constraintSetup {angularLimitsMax=<0 5 10>, angularLimitsMin=<0 -30 -10>, angularTarget=<0 0 0>}
      localJointOffset=<3 0 0>
     }]
    relativeSpaceBone {boneName=root}
   }
   n12 = Anim Dynamics - Bone: jacket_bck_l_02(ComponentPose=@n11, Alpha=1, ExternalForce="0,0,0") {
    angularSpringConstant=2000
    bAngularSpring=true
    bGravityOverrideInSimSpace=true
    bUseGravityOverride=true
    simulationSpace=RootRelative
    alphaBoolBlend {blendOption=Linear}
    boundBone {boneName=jacket_bck_l_02}
    componentLinearAccScale=<1 1 1>
    componentLinearVelScale=<1 1 1>
    gravityOverride {z=-1500}
    physicsBodyDefinitions [{
      boundBone {boneName=jacket_bck_l_02}
      boxExtents=<30 30 30>
      constraintSetup {angularLimitsMax=<0 5 10>, angularLimitsMin=<0 -50 -10>, angularTarget=<0 0 0>}
      localJointOffset=<3 0 0>
     }]
    relativeSpaceBone {boneName=root}
   }
   n13 = Anim Dynamics - Bone: jacket_side_l_01(ComponentPose=@n12, Alpha=1, ExternalForce="0,0,0") {
    angularSpringConstant=2000
    bAngularSpring=true
    bGravityOverrideInSimSpace=true
    bUseGravityOverride=true
    simulationSpace=RootRelative
    alphaBoolBlend {blendOption=Linear}
    boundBone {boneName=jacket_side_l_01}
    componentLinearAccScale=<1 1 1>
    componentLinearVelScale=<1 1 1>
    gravityOverride {z=-1500}
    physicsBodyDefinitions [{
      boundBone {boneName=jacket_side_l_01}
      boxExtents=<30 30 30>
      constraintSetup {angularLimitsMax=<0 5 10>, angularLimitsMin=<0 -45 -10>, angularTarget=<0 0 0>}
      localJointOffset=<3 0 0>
     }]
    relativeSpaceBone {boneName=root}
   }
   Anim Dynamics - Bone: jacket_side_l_02(ComponentPose=@n13, Alpha=1, ExternalForce="0,0,0") {
    angularSpringConstant=2000
    bAngularSpring=true
    bGravityOverrideInSimSpace=true
    bUseGravityOverride=true
    simulationSpace=RootRelative
    alphaBoolBlend {blendOption=Linear}
    boundBone {boneName=jacket_side_l_02}
    componentLinearAccScale=<1 1 1>
    componentLinearVelScale=<1 1 1>
    gravityOverride {z=-1500}
    physicsBodyDefinitions [{
      boundBone {boneName=jacket_side_l_02}
      boxExtents=<30 30 30>
      constraintSetup {angularLimitsMax=<0 5 10>, angularLimitsMin=<0 -50 -10>, angularTarget=<0 0 0>}
      localJointOffset=<3 0 0>
     }]
    relativeSpaceBone {boneName=root}
   }
   n14 = "Constraint - kneepad_l (PP)"!(ComponentPose, Alpha=1, ConstraintWeights_0=0.2, ConstraintWeights_1:float=0.8) -> (Pose) {
    alphaBoolBlend {blendOption=Linear}
    boneToModify {boneName=kneepad_l}
    constraintSetup [{targetBone {boneName=thigh_l}}, {targetBone {boneName=calf_l}}]
    constraintWeights [1,1]
   }
   n15 = "Constraint - kneepad_r (PP)"!(ComponentPose=@n14, Alpha=1, ConstraintWeights_0=0.2, ConstraintWeights_1:float=0.8) -> (Pose) {
    alphaBoolBlend {blendOption=Linear}
    boneToModify {boneName=kneepad_r}
    constraintSetup [{targetBone {boneName=thigh_r}}, {targetBone {boneName=calf_r}}]
    constraintWeights [1,1]
   }
   n16 = "Constraint - dyn_knee_l (PP)"!(ComponentPose=@n15, Alpha=1, ConstraintWeights_0=0.5, ConstraintWeights_1:float=0.5) -> (Pose) {
    alphaBoolBlend {blendOption=Linear}
    boneToModify {boneName=dyn_knee_l}
    constraintSetup [{targetBone {boneName=thigh_l}}, {targetBone {boneName=calf_l}}]
    constraintWeights [1,1]
   }
   "Constraint - dyn_knee_r (PP)"!(ComponentPose=@n16, Alpha=1, ConstraintWeights_0=0.5, ConstraintWeights_1:float=0.5) -> (Pose) {
    alphaBoolBlend {blendOption=Linear}
    boneToModify {boneName=dyn_knee_r}
    constraintSetup [{targetBone {boneName=thigh_r}}, {targetBone {boneName=calf_r}}]
    constraintWeights [1,1]
   }
   n17 = Control Rig!(Source=Copy Pose From Mesh() {bUseAttachedParent=true}, Alpha, is in Frontend:bool=Get bIsInFrontEnd()) -> (Pose) {
    alphaBoolBlend {blendOption=Linear}
    controlRigAssetReference {blueprintRigClass=&/F_LRG_DeformRig_SpaceFeline_ControlRig.F_LRG_DeformRig_SpaceFeline_ControlRig_C}
   }
   n18 = "Transform (Modify) Bone - Bone: tube_r_01"(ComponentPose=Local To Component(LocalPose=@n17), Alpha=1, Translation="1.199,2.402,-0.000", Rotation="0,0,0", Scale="1,1,1") {translationMode=BMM_Additive, alphaBoolBlend {blendOption=Linear}, boneToModify {boneName=tube_r_01}, translation=<1.199 2.402>}
   n19 = Anim Dynamics - Bone: tube_r_01(ComponentPose=@n18, Alpha=0.5, ExternalForce="0,0,0") {
    angularSpringConstant=500
    bAngularSpring=true
    bGravityOverrideInSimSpace=true
    bUseGravityOverride=true
    linearSpringConstant=0.8
    simulationSpace=RootRelative
    alphaBoolBlend {blendOption=Linear}
    boundBone {boneName=tube_r_01}
    gravityOverride {z=-1500}
    physicsBodyDefinitions [{
      boundBone {boneName=tube_r_01}
      boxExtents=<20 20 20>
      constraintSetup {angularLimitsMax=<0 5 2>, angularLimitsMin=<0 -4 -2>, angularTarget=<0 0 0>}
      localJointOffset=<0.4 0 0>
     }]
    relativeSpaceBone {boneName=root}
   }
   n20 = Anim Dynamics - Bone: tube_r_02(ComponentPose=@n19, Alpha=0.5, ExternalForce="0,0,0") {
    angularSpringConstant=500
    bAngularSpring=true
    bGravityOverrideInSimSpace=true
    bUseGravityOverride=true
    linearSpringConstant=0.8
    simulationSpace=RootRelative
    alphaBoolBlend {blendOption=Linear}
    boundBone {boneName=tube_r_02}
    gravityOverride {z=-1500}
    physicsBodyDefinitions [{
      boundBone {boneName=tube_r_02}
      boxExtents=<20 20 20>
      constraintSetup {angularLimitsMax=<0 2 0.5>, angularLimitsMin=<0 -2 -0.5>, angularTarget=<0 0 0>}
      localJointOffset=<0.4 0 0>
     }]
    relativeSpaceBone {boneName=root}
   }
   n21 = Anim Dynamics - Bone: tube_l_01(ComponentPose=@n20, Alpha=0.5, ExternalForce="0,0,0") {
    angularSpringConstant=500
    bAngularSpring=true
    bGravityOverrideInSimSpace=true
    bUseGravityOverride=true
    linearSpringConstant=0.8
    simulationSpace=RootRelative
    alphaBoolBlend {blendOption=Linear}
    boundBone {boneName=tube_l_01}
    gravityOverride {z=-1500}
    physicsBodyDefinitions [{
      boundBone {boneName=tube_l_01}
      boxExtents=<20 20 20>
      constraintSetup {angularLimitsMax=<0 5 2>, angularLimitsMin=<0 -1 -2>, angularTarget=<0 0 0>}
      localJointOffset=<0.4 0 0>
     }]
    relativeSpaceBone {boneName=root}
   }
   n22 = Anim Dynamics - Bone: tube_l_02(ComponentPose=@n21, Alpha=0.5, ExternalForce="0,0,0") {
    angularSpringConstant=500
    bAngularSpring=true
    bGravityOverrideInSimSpace=true
    bUseGravityOverride=true
    linearSpringConstant=0.8
    simulationSpace=RootRelative
    alphaBoolBlend {blendOption=Linear}
    boundBone {boneName=tube_l_02}
    gravityOverride {z=-1500}
    physicsBodyDefinitions [{
      boundBone {boneName=tube_l_02}
      boxExtents=<20 20 20>
      constraintSetup {angularLimitsMax=<0 2 0.5>, angularLimitsMin=<0 -2 -0.5>, angularTarget=<0 0 0>}
      localJointOffset=<0.4 0 0>
     }]
    relativeSpaceBone {boneName=root}
   }
   n23 = RigidBody(ComponentPose=@n22) {
    bClampLinearTranslationLimitToRefPose=true
    bOverrideWorldGravity=true
    overridePhysicsAsset=&/F_LRG_SpaceFeline_Physics
    simulationSpace=BaseBoneSpace
    alphaBoolBlend {blendOption=Linear}
    baseBoneRef {boneName=spine_05}
    componentLinearAccScale=<0.35 0.35 0.35>
    componentLinearVelScale=<0.35 0.35 0.35>
    overrideWorldGravity {z=-980}
   }
   Output Pose(Result=Component To Local(ComponentPose=@n23)) {layerGroup=None}
  ]}
 EventGraph {nodes [Event BlueprintUpdateAnimation()]}
]
variables []
