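# Animation Blueprint for the SpaceFeline face accessory, posing its cat arm and joystick bones on the player character.
# A Control Rig refines the parent mesh's copied pose using clamped directional movement values plus skydive, parachute and emote flags.
# In the front end, the Frontend_Idle_Bypass curve weights a per-bone blend of DefaultSlot animation onto those bones.
# Constraint nodes for hands, elbows and joysticks are disconnected from the output.
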AnimBlueprint /BRCosmetics/Characters/Player/Female/Large/Bodies/F_LRG_SpaceFeline/Meshes/Parts/F_LRG_SpaceFeline_FaceAcc_AnimBP
target_skeleton=&/F_LRG_SpaceFeline_FaceAcc_Skeleton
graphs [
 AnimGraph {nodes [
   "Constraint - joystick_r (PP)"!(ComponentPose, Alpha=1, ConstraintWeights_0=0.5, ConstraintWeights_1:float=0.5) -> (Pose) {
    alphaBoolBlend {blendOption=Linear}
    boneToModify {boneName=joystick_r}
    constraintSetup [{targetBone {boneName=neck_01}}, {targetBone {boneName=spine_05}}]
    constraintWeights [1,1]
   }
   "Constraint - joystick_l (PP)"!(ComponentPose, Alpha=1, ConstraintWeights_0=0.5, ConstraintWeights_1:float=0.5) -> (Pose) {
    alphaBoolBlend {blendOption=Linear}
    boneToModify {boneName=joystick_l}
    constraintSetup [{targetBone {boneName=neck_01}}, {targetBone {boneName=spine_05}}]
    constraintWeights [1,1]
   }
   n1 = "Constraint - cat_elbow_l (PP)"!(ComponentPose, Alpha=0.5, ConstraintWeights_0=1, ConstraintWeights_1:float=0) -> (Pose) {
    alphaBoolBlend {blendOption=Linear}
    boneToModify {boneName=cat_elbow_l}
    constraintSetup [{targetBone {boneName=neck_01}}, {targetBone {boneName=cat_hand_l}}]
    constraintWeights [1,1]
   }
   "Constraint - cat_elbow_r (PP)"!(ComponentPose=@n1, Alpha=0.5, ConstraintWeights_0=1, ConstraintWeights_1:float=0) -> (Pose) {
    alphaBoolBlend {blendOption=Linear}
    boneToModify {boneName=cat_elbow_r}
    constraintSetup [{targetBone {boneName=cat_hand_r}}, {targetBone {boneName=neck_01}}]
    constraintWeights [1,1]
   }
   n2 = "Constraint - cat_hand_l (P)"(Alpha=1, ConstraintWeights_0=1) {
    alphaBoolBlend {blendOption=Linear}
    boneToModify {boneName=cat_hand_l}
    constraintSetup [{targetBone {boneName=joystick_l}}]
    constraintWeights [1]
   }
   "Constraint - cat_hand_r (P)"(ComponentPose=@n2, Alpha=1, ConstraintWeights_0=1) {
    alphaBoolBlend {blendOption=Linear}
    boneToModify {boneName=cat_hand_r}
    constraintSetup [{targetBone {boneName=joystick_r}}]
    constraintWeights [1]
   }
   n3 = Control Rig!(
    Source=Copy Pose From Mesh() {bCopyCurves=true, bUseAttachedParent=true}
    Alpha
    Character Forward:double=Get Character Forward() -> (Character Forward:double)
    Character Backward:double=Get Character Backward() -> (Character Backward:double)
    Character Move Right:double=Get Character Move Right() -> (Character Move Right:double)
    Character Move Left:double=Get Character Move Left() -> (Character Move Left:double)
    Is Skydiving:bool=Get bIsSkyDiving()
    Is Parachute Opened:bool=Get bIsParachuteOpened()
    Is Playing Emote:bool=Get bIsPlayingEmote()
   ) -> (Pose) {
    alphaBoolBlend {blendOption=Linear}
    controlRigAssetReference {blueprintRigClass=&/F_LRG_SpaceFeline_FaceAcc_CtrlRig.F_LRG_SpaceFeline_FaceAcc_CtrlRig_C}
   }
   n4 = Layered blend per bone(
    BasePose=Component To Local(ComponentPose=Local To Component(LocalPose=@n3))
    BlendPoses_0=Slot 'DefaultSlot'()
    BlendWeights_0=Get IsPlayingFrontendIdle() -> (IsPlayingFrontendIdle:float)
   ) {layerSetup [{branchFilters [{boneName=joystick_l}, {boneName=joystick_r}, {boneName=cat_hand_l}, {boneName=cat_elbow_l}, {boneName=cat_hand_r}, {boneName=cat_elbow_r}]}]}
   Output Pose(Result=@n4) {layerGroup=None}
  ]}
 EventGraph {nodes [
   Event BlueprintInitializeAnimation()
   > Set SkelMesh(SkelMesh:Skeletal Mesh Object Reference=/Game/Accessories/FORT_Backpacks/Backpack_F_MED_Skirmish/Meshes/F_MED_Skirmish_Pack) -> (Output_Get:Skeletal Mesh Object Reference)
   > n1 = Cast To FortPlayerPawn(Object=TryGetFortPawn())
   > Set Pawn(Pawn:Fort Player Pawn Object Reference=@n1) -> (Output_Get:Fort Player Pawn Object Reference)
   Event BlueprintUpdateAnimation()
   > Get!(execute:Exec) -> (then:Exec, else:Exec, SkelMesh:Skeletal Mesh Object Reference)
   > Get!(execute) -> (then, else, Pawn:Fort Player Pawn Object Reference)
   > n2 = Sequence()
   > n3 = PlayerMovement(AnimInstance=Self-Reference(), XDotProductCoefficient=25, YDotProductCoefficient=25, ZDotProductCoefficient=1)
   > Set Character Move Right(Character Move Right:double=MapRangeClamped(Value=@n3, InRangeA=0, InRangeB=25)) -> (Output_Get:double)
   > Set Character Move Left(Character Move Left:double=MapRangeClamped(Value=@n3, InRangeA=0, InRangeB=-25)) -> (Output_Get:double)
   > Set Character Forward(Character Forward:double=MapRangeClamped(Value=@n3.YMovement, InRangeA=0, InRangeB=25)) -> (Output_Get:double)
   > Set Character Backward(Character Backward:double=MapRangeClamped(Value=@n3.YMovement, InRangeA=0, InRangeB=-25)) -> (Output_Get:double)
   Branch(execute=@n2.then_1, Condition=Get bIsInFrontEnd())
   > Set IsPlayingFrontendIdle(IsPlayingFrontendIdle:float=GetCurveValue(self=GetOwnerFrontEndAnimInstance(), CurveName=Frontend_Idle_Bypass)) -> (Output_Get:float)
   Get PawnVelocity()
  ]}
]
variables [
 Character Backward {type=number}
 Character Forward {type=number}
 Character Move Left {type=number}
 Character Move Right {type=number}
 IsPlayingFrontendIdle {type=number}
 Pawn {type=FortPlayerPawn}
 SkelMesh {type=SkeletalMesh, default=&F_MED_Skirmish_Pack}
]
